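; BP_Athena_Event_Components is a Blueprint that previews Athena map event dressing: a rainbow and a St. Patrick's river tint.
; Its construction script applies the rainbow's visibility from an in-editor testing flag and places it at a stored world transform.
; It then calls SetupRiverTint, which writes the RiverTintAdd vector in the MPC_Landscape material parameter collection, using the St. Patrick's color when enabled and a zero-RGB color otherwise.
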(fn SetupRiverTint
  (params (EnableTint :type Boolean) (TintColor :type Linear%20Color%20Structure))
  (if EnableTint
    (Rendering|Material|SetVectorParameterValue "/Game/Athena/Environments/Landscape/MPC/MPC_Landscape.MPC_Landscape" "RiverTintAdd" TintColor :!class /Script/Engine.KismetMaterialLibrary)
    (else
      (Rendering|Material|SetVectorParameterValue "/Game/Athena/Environments/Landscape/MPC/MPC_Landscape.MPC_Landscape" "RiverTintAdd" (Utilities|Struct|MakeLinearColor 0.0) :!class /Script/Engine.KismetMaterialLibrary))))

(fn ConstructionScript
  (bind _rainbow (Variables|Default|GetRainbow))
  (Utilities|FlowControl|Sequence
    (:then_1
      (Rendering|SetVisibility _rainbow (Variables|InEditorTesting|GetShowRainbow) false)
      (Transformation|SetWorldTransform _rainbow (Variables|Default|GetRainbowWorldTransform) false false))
    (:then_2
      (CallFunction|SetupRiverTint (Variables|InEditorTesting|GetEnableRiverTint) (Variables|Default|GetRiverTintStPatrick's)))))
